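# Animation Blueprint that adds procedural hair motion to the female Starfish head 01 on the Fortnite base head skeleton.
# The input pose passes through chained Anim Dynamics nodes: five angular-spring, heavily damped, limit-constrained bodies for the dyn_hair_05 and dyn_hair_04 bones.
# A final linear-spring node jiggles dyn_hair_01 in actor space at very low alpha, with reduced gravity and an LOD threshold.
# The event graph contains no logic.
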AnimBlueprint /Game/Characters/Player/Female/Medium/Heads/F_MED_ASN_Starfish_Head_01/Meshes/F_MED_ASN_Starfish_Head_01_dyn_AnimBP
target_skeleton=/Game/Characters/Player/Common/Fortnite_Base_Head/Fortnite_Base_Head_Export_Skeleton
graphs [
 AnimGraph {nodes [
   AnimGraphNode_SubInput_1 = Input Pose()
   n1 = Anim Dynamics - Bone: dyn_hair_05(ComponentPose=Local To Component(LocalPose=@AnimGraphNode_SubInput_1), Alpha=1, ExternalForce="0,0,0") {
    angularDampingOverride=0.99
    angularSpringConstant=10000
    bAngularSpring=true
    bGravityOverrideInSimSpace=true
    bOverrideAngularDamping=true
    numSolverIterationsPostUpdate=4
    numSolverIterationsPreUpdate=16
    alphaBoolBlend {blendOption=Linear}
    boundBone {boneName=dyn_hair_05}
    physicsBodyDefinitions [{
      sphereCollisionRadius=0
      boundBone {boneName=dyn_hair_05}
      boxExtents=<5 30 30>
      constraintSetup {angularTargetAxis=AxisZ, angularLimitsMax=<2 20 0>, angularLimitsMin=<-2 -5 -20>, angularTarget=<0 0 1>}
      localJointOffset=<-1 0 0>
     }]
   }
   n2 = Anim Dynamics - Bone: dyn_hair_05_a(ComponentPose=@n1, Alpha=1, ExternalForce="0,0,0") {
    angularDampingOverride=0.99
    angularSpringConstant=10000
    bAngularSpring=true
    bGravityOverrideInSimSpace=true
    bOverrideAngularDamping=true
    numSolverIterationsPostUpdate=4
    numSolverIterationsPreUpdate=16
    alphaBoolBlend {blendOption=Linear}
    boundBone {boneName=dyn_hair_05_a}
    physicsBodyDefinitions [{
      sphereCollisionRadius=0
      boundBone {boneName=dyn_hair_05_a}
      boxExtents=<5 30 30>
      constraintSetup {angularTargetAxis=AxisZ, angularLimitsMax=<2 20 10>, angularLimitsMin=<-2 -6 -20>, angularTarget=<0 0 1>}
      localJointOffset=<-1 0 0>
     }]
   }
   n3 = Anim Dynamics - Bone: dyn_hair_05_b(ComponentPose=@n2, Alpha=1, ExternalForce="0,0,0") {
    angularDampingOverride=0.99
    angularSpringConstant=10000
    bAngularSpring=true
    bGravityOverrideInSimSpace=true
    bOverrideAngularDamping=true
    numSolverIterationsPostUpdate=4
    numSolverIterationsPreUpdate=16
    alphaBoolBlend {blendOption=Linear}
    boundBone {boneName=dyn_hair_05_b}
    physicsBodyDefinitions [{
      sphereCollisionRadius=0
      boundBone {boneName=dyn_hair_05_b}
      boxExtents=<5 30 30>
      constraintSetup {angularTargetAxis=AxisZ, angularLimitsMax=<2 20 20>, angularLimitsMin=<-2 -20 -20>, angularTarget=<0 0 1>}
      localJointOffset=<-1 0 0>
     }]
   }
   n4 = Anim Dynamics - Bone: dyn_hair_04(ComponentPose=@n3, Alpha=1, ExternalForce="0,0,0") {
    angularDampingOverride=0.99
    angularSpringConstant=10000
    bAngularSpring=true
    bGravityOverrideInSimSpace=true
    bOverrideAngularDamping=true
    numSolverIterationsPostUpdate=4
    numSolverIterationsPreUpdate=16
    alphaBoolBlend {blendOption=Linear}
    boundBone {boneName=dyn_hair_04}
    physicsBodyDefinitions [{
      sphereCollisionRadius=0
      boundBone {boneName=dyn_hair_04}
      boxExtents=<5 30 30>
      constraintSetup {angularTargetAxis=AxisZ, angularLimitsMax=<2 20 20>, angularLimitsMin=<-2 -20 -20>, angularTarget=<0 0 1>}
      localJointOffset=<-1 0 0>
     }]
   }
   n5 = Anim Dynamics - Bone: dyn_hair_04_a(ComponentPose=@n4, Alpha=1, ExternalForce="0,0,0") {
    angularDampingOverride=0.99
    angularSpringConstant=10000
    bAngularSpring=true
    bGravityOverrideInSimSpace=true
    bOverrideAngularDamping=true
    numSolverIterationsPostUpdate=4
    numSolverIterationsPreUpdate=16
    alphaBoolBlend {blendOption=Linear}
    boundBone {boneName=dyn_hair_04_a}
    physicsBodyDefinitions [{
      sphereCollisionRadius=0
      boundBone {boneName=dyn_hair_04_a}
      boxExtents=<5 30 30>
      constraintSetup {angularTargetAxis=AxisZ, angularLimitsMax=<2 20 20>, angularLimitsMin=<-2 -20 -20>, angularTarget=<0 0 1>}
      localJointOffset=<-1 0 0>
     }]
   }
   n6 = Anim Dynamics - Bone: dyn_hair_01(ComponentPose=@n5, Alpha=0.05) {
    bGravityOverrideInSimSpace=true
    bLinearSpring=true
    bOverrideLinearDamping=true
    gravityScale=0.2
    lODThreshold=1
    linearDampingOverride=0.997
    linearSpringConstant=200
    simulationSpace=Actor
    alphaBoolBlend {blendOption=Linear}
    boundBone {boneName=dyn_hair_01}
    physicsBodyDefinitions [{
      sphereCollisionRadius=0
      boundBone {boneName=dyn_hair_01}
      boxExtents=<60 60 60>
      constraintSetup {
       twistAxis=AxisZ
       angularLimitsMax=<0 0 0>
       angularLimitsMin=<0 0 0>
       angularTarget=<0 0 0>
       linearAxesMax=<10 10 10>
       linearAxesMin=<-10 -10 -10>
      }
      localJointOffset=<0 0 0>
     }]
   }
   Output Pose(Result=Component To Local(ComponentPose=@n6)) {layerGroup=None}
  ]}
 EventGraph {nodes [
   TryGetPawnOwner()
   Event BlueprintUpdateAnimation()
  ]}
]
variables []
